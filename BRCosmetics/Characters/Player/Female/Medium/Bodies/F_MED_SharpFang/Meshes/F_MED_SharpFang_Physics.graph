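# Physics asset for the SharpFang medium female player body, giving capsule bodies to its hanging accessories, skirt, pelvis and thighs.
# A three-link side chain, back wolf head and bag hang from the pelvis with limited swing, and two small pieces hang off the bag.
# Skirt A and B panels constrain to the thighs, C to the pelvis.
# Thigh constraints are fully locked, though thigh_l has none to the pelvis.
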PhysicsAsset /BRCosmetics/Characters/Player/Female/Medium/Bodies/F_MED_SharpFang/Meshes/F_MED_SharpFang_Physics
bodies {
 dyn_Chain_Side_1 [{shapeType=Capsule, length=10.005, radius=1.255, shapeName=dyn_Chain_Side_1_capsule, center=<5.2656 0.2846 0.1341>, rotation=<rot 0 -90 0>}]
 dyn_Chain_Side_2 [{
   shapeType=Capsule
   length=7.3003
   radius=1.5476
   shapeName=dyn_Chain_Side_2_capsule
   center=<3.7742 0.0654 -0.0546>
   rotation=<rot -0.427 -90 -0.4662>
  }]
 dyn_Chain_Side_3 [{
   shapeType=Capsule
   length=3.0246
   radius=1.6808
   shapeName=dyn_Chain_Side_3_capsule
   center=<4.016 -0.8506 -0.0877>
   rotation=<rot 0.4159 -97.5369 -87.0794>
  }]
 dyn_Skirt_side_A_01_l [{
   shapeType=Capsule
   length=13.0265
   radius=2.3693
   shapeName=dyn_Skirt_side_A_01_l_capsule
   center=<7.4047 -1.098 2.203>
   rotation=<rot 6.3703 -90.8151 -9.9633>
  }]
 dyn_Skirt_side_A_01_r [{
   shapeType=Capsule
   length=11.8563
   radius=1.6432
   shapeName=dyn_Skirt_side_A_01_r_capsule
   center=<7.0444 0.2512 1.3355>
   rotation=<rot 1.2364 -89.9999 -5.8777>
  }]
 dyn_Skirt_side_B_01_l [{
   shapeType=Capsule
   length=13.7852
   radius=2.1437
   shapeName=dyn_Skirt_side_B_01_l_capsule
   center=<7.8435 1.5039 0.7375>
   rotation=<rot -0.13 -90 -7.0169>
  }]
 dyn_Skirt_side_B_01_r [{
   shapeType=Capsule
   length=9.2072
   radius=1.6252
   shapeName=dyn_Skirt_side_B_01_r_capsule
   center=<6.8115 -0.7753 0.0363>
   rotation=<rot -2.2376 -89.9554 -0.8534>
  }]
 dyn_Skirt_side_C_01_l [{
   shapeType=Capsule
   length=13.5191
   radius=2.6459
   shapeName=dyn_Skirt_side_C_01_l_capsule
   center=<6.6637 -0.6395 -1.9015>
   rotation=<rot 0.7355 -89.5282 2.1152>
  }]
 dyn_Skirt_side_C_01_r [{
   shapeType=Capsule
   length=12.58
   radius=2.7361
   shapeName=dyn_Skirt_side_C_01_r_capsule
   center=<6.9934 1.516 -0.6841>
   rotation=<rot 0.7616 -89.9572 0.5149>
  }]
 dyn_back_HeadWolf [{
   shapeType=Capsule
   length=2.25
   radius=2.8433
   shapeName=dyn_back_HeadWolf_capsule
   center=<7.5318 -1.4063 0.0915>
   rotation=<rot -0.2275 -78.589 -89.1421>
  }]
 dyn_bag [{
   shapeType=Capsule
   length=8.3091
   radius=5.2074
   shapeName=dyn_bag_capsule
   center=<5.9684 -2.3963 1.8978>
   rotation=<rot 9.6469 -90 -25.2736>
  }]
 dyn_bag_dop [{
   shapeType=Capsule
   length=4.1687
   radius=1.1429
   shapeName=dyn_bag_dop_capsule
   center=<2.6727 -0.2551 -0.0735>
   rotation=<rot 0.2116 -94.0367 -87.0777>
  }]
 dyn_bag_zip [{
   shapeType=Capsule
   length=1.8034
   radius=0.6273
   shapeName=dyn_bag_zip_capsule
   center=<1.6716 0.0031 0.0101>
   rotation=<rot 0 -89.7287 -87.3635>
  }]
 dyn_thigh_twist_01_corr_l [
  {
   shapeType=Capsule
   length=33.76
   radius=8.25
   shapeName=dyn_thigh_twist_01_corr_r_capsule
   center=<15.4467 0.9577 0.1775>
   rotation=<rot 0.2025 -95.6975 -95.0801>
  }
  {
   shapeType=Capsule
   length=15
   radius=4
   shapeName=dyn_thigh_twist_01_corr_r_capsule
   center=<10.8389 3.7488 -5.4616>
   rotation=<rot -78.7951 1.0556 178.1608>
  }
 ]
 dyn_thigh_twist_01_corr_r [{
   shapeType=Capsule
   length=20.51
   radius=8
   shapeName=dyn_thigh_twist_01_corr_r_capsule
   center=<-13.486 -0.4504 -0.2666>
   rotation=<rot 0.3066 -94.5241 86.7392>
  }]
 pelvis [{
   shapeType=Capsule
   length=7.4867
   radius=5.3568
   shapeName=pelvis_capsule
   center=<5.5833 -4.4877 -3.6964>
   rotation=<rot -13.1593 -75.8093 -2.1713>
  }]
 thigh_l [{
   shapeType=Capsule
   length=20.75
   radius=9.25
   shapeName=thigh_l_capsule
   center=<-20.0557 -0.9537 2.4997>
   rotation=<rot -89.1573 -180 179.9998>
  }]
 thigh_r [{shapeType=Capsule, length=1.5, radius=6.5, shapeName=thigh_r_capsule, center=<1.5133 -0.5107 0.4091>, rotation=<rot 0 -0.0001 -104.5848>}]
 thigh_twist_01_r [{
   shapeType=Capsule
   length=16.435
   radius=9.4652
   shapeName=thigh_twist_01_r_capsule
   center=<-26.7746 -1.3977 -4.2009>
   rotation=<rot -75.645 37.4893 -41.9094>
  }]
}
constraints [
 {bone1Name=dyn_bag_zip, bone2Name=dyn_bag, swing1LimitDegrees=25, swing2LimitDegrees=5, twistLimitDegrees=45, twistMotion=Locked}
 {bone1Name=dyn_Chain_Side_1, bone2Name=pelvis, swing1LimitDegrees=15, swing2LimitDegrees=12, twistLimitDegrees=3}
 {bone1Name=dyn_Chain_Side_2, bone2Name=dyn_Chain_Side_1, swing1LimitDegrees=25, swing2LimitDegrees=12, twistLimitDegrees=15}
 {bone1Name=dyn_Chain_Side_3, bone2Name=dyn_Chain_Side_2, swing1LimitDegrees=45, swing2LimitDegrees=12, twistLimitDegrees=10}
 {bone1Name=dyn_back_HeadWolf, bone2Name=pelvis, swing1LimitDegrees=20, swing2LimitDegrees=13, twistLimitDegrees=5}
 {bone1Name=dyn_bag, bone2Name=pelvis, swing1LimitDegrees=10, swing2LimitDegrees=10, twistLimitDegrees=5, twistMotion=Locked}
 {bone1Name=dyn_bag_dop, bone2Name=dyn_bag, swing1LimitDegrees=35, swing2LimitDegrees=15, twistLimitDegrees=5}
 {bone1Name=dyn_Skirt_side_B_01_r, bone2Name=thigh_r, swing1LimitDegrees=10, swing2LimitDegrees=10, twistLimitDegrees=10}
 {bone1Name=dyn_Skirt_side_A_01_r, bone2Name=thigh_r, swing1LimitDegrees=10, swing2LimitDegrees=10, twistLimitDegrees=10}
 {
  bone1Name=thigh_r
  bone2Name=pelvis
  swing1LimitDegrees=45
  swing1Motion=Locked
  swing2LimitDegrees=45
  swing2Motion=Locked
  twistLimitDegrees=45
  twistMotion=Locked
 }
 {
  bone1Name=thigh_twist_01_r
  bone2Name=thigh_r
  swing1LimitDegrees=45
  swing1Motion=Locked
  swing2LimitDegrees=45
  swing2Motion=Locked
  twistLimitDegrees=45
  twistMotion=Locked
 }
 {
  bone1Name=dyn_thigh_twist_01_corr_r
  bone2Name=thigh_twist_01_r
  swing1LimitDegrees=45
  swing1Motion=Locked
  swing2LimitDegrees=45
  swing2Motion=Locked
  twistLimitDegrees=45
  twistMotion=Locked
 }
 {
  bone1Name=dyn_thigh_twist_01_corr_l
  bone2Name=pelvis
  swing1LimitDegrees=45
  swing1Motion=Locked
  swing2LimitDegrees=45
  swing2Motion=Locked
  twistLimitDegrees=45
  twistMotion=Locked
 }
 {bone1Name=dyn_Skirt_side_C_01_l, bone2Name=pelvis, swing1LimitDegrees=15, swing2LimitDegrees=10, twistLimitDegrees=5}
 {bone1Name=dyn_Skirt_side_C_01_r, bone2Name=pelvis, swing1LimitDegrees=15, swing2LimitDegrees=10, twistLimitDegrees=5}
 {bone1Name=dyn_Skirt_side_A_01_l, bone2Name=thigh_l, swing1LimitDegrees=10, swing2LimitDegrees=10, twistLimitDegrees=10}
 {bone1Name=dyn_Skirt_side_B_01_l, bone2Name=thigh_l, swing1LimitDegrees=10, swing2LimitDegrees=10, twistLimitDegrees=10}
]
