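# Physics asset for the SharpFang female medium player cosmetic body, simulating two dangling bangles on the left forearm.
# Each bangle bone has a box body, and the lowerarm_l bone has a capsule.
# Two constraints tie the bangles directly to the forearm with narrow swing and twist limits, keeping their motion close to the arm.
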PhysicsAsset /BRCosmetics/Characters/Player/Female/Medium/Bodies/F_MED_SharpFang/Meshes/F_MED_SharpFang_Physics1
bodies {
 dyn_bangle_01 [{
   shapeType=Box
   extentX=6.0223
   extentY=1.7496
   extentZ=9.5205
   shapeName=dyn_bangle_01_box
   center=<-0.0007 -0.0641 -0.1649>
   rotation=<rot 21.2583 -90 0.2387>
  }]
 dyn_bangle_02 [{
   shapeType=Box
   extentX=8.8517
   extentY=2.5776
   extentZ=8.8281
   shapeName=dyn_bangle_02_box
   center=<0.0111 0.0921 0.2578>
   rotation=<rot 19.6785 -90 2.3245>
  }]
 lowerarm_l [{shapeType=Capsule, length=15.234, radius=2.7494, center=<14.6375 -0.1525 0.1452>, rotation=<rot -0.1255 90.1592 -93.1529>}]
}
constraints [
 {bone1Name=dyn_bangle_01, bone2Name=lowerarm_l, swing1LimitDegrees=15, swing2LimitDegrees=15, twistLimitDegrees=10}
 {bone1Name=dyn_bangle_02, bone2Name=lowerarm_l, swing1LimitDegrees=20, swing2LimitDegrees=20, twistLimitDegrees=5}
]
