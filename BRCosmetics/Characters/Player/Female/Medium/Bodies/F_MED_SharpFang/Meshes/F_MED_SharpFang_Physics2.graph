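# Physics asset for the F_MED_SharpFang female player cosmetic body, covering its two dynamic neck chain bones.
# Capsules give shape to both chain bones, with two on dyn_Chain_Neck_1, and to spine_05 at the upper torso.
# Each chain bone is constrained directly to spine_05 rather than to the other, with narrow, asymmetric swing limits.
# The dyn_Chain_Neck_B constraint also locks twist.
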PhysicsAsset /BRCosmetics/Characters/Player/Female/Medium/Bodies/F_MED_SharpFang/Meshes/F_MED_SharpFang_Physics2
bodies {
 dyn_Chain_Neck_1 [
  {
   shapeType=Capsule
   length=4.6976
   radius=1.7167
   shapeName=dyn_Chain_Neck_1_capsule
   center=<14.8945 -0.6194 -0.2771>
   rotation=<rot -0.1808 -90 0.1884>
  }
  {shapeType=Capsule, length=9.75, radius=0.75, shapeName=dyn_Chain_Neck_1_capsule, center=<11.349 0 0>}
 ]
 dyn_Chain_Neck_B [{
   shapeType=Capsule
   length=4.4389
   radius=1.0215
   shapeName=dyn_Chain_Neck_B_capsule
   center=<9.5374 0.3319 -0.0165>
   rotation=<rot -0.122 -90 0.4613>
  }]
 spine_05 [{
   shapeType=Capsule
   length=11.5938
   radius=11.6345
   shapeName=spine_05_capsule
   center=<-4.3706 -1.2572 -0.6231>
   rotation=<rot -4.1613 -124.8875 -90.7958>
  }]
}
constraints [
 {bone1Name=dyn_Chain_Neck_1, bone2Name=spine_05, swing1LimitDegrees=20, swing2LimitDegrees=2, twistLimitDegrees=2}
 {bone1Name=dyn_Chain_Neck_B, bone2Name=spine_05, swing1LimitDegrees=14, swing2LimitDegrees=2, twistLimitDegrees=45, twistMotion=Locked}
]
