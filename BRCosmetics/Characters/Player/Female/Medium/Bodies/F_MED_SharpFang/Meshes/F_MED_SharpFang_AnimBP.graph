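# Post-process style Animation Blueprint for the F_MED_SharpFang female body cosmetic, adding secondary motion and pose correctives to the copied parent pose.
# Three RigidBody sims and sleeve AnimDynamics feed additive bone offsets for sleeves, bag, skirt and chain, then a Control Rig.
# The update event derives offset alphas from arm swing, crouch direction and yaw delta.
# It also flips gravity upward while falling and zero-scales extra_pel_scale above LOD 1.
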AnimBlueprint /BRCosmetics/Characters/Player/Female/Medium/Bodies/F_MED_SharpFang/Meshes/F_MED_SharpFang_AnimBP
target_skeleton=&/F_MED_SharpFang_Skeleton
graphs [
 AnimGraph {nodes [
   n1 = Get Gravity Vector() -> (Gravity Vector:Vector)
   n2 = RigidBody!(
    ComponentPose=Local To Component(LocalPose=Copy Pose From Mesh() {bUseAttachedParent=true})
    Alpha
    OverrideWorldGravity:Vector=@n1
    ExternalForce
   ) -> (Pose) {
    bClampLinearTranslationLimitToRefPose=true
    bOverrideWorldGravity=true
    overridePhysicsAsset=&/F_MED_SharpFang_Physics
    simulationSpace=BaseBoneSpace
    alphaBoolBlend {blendOption=Linear}
    baseBoneRef {boneName=pelvis}
    componentLinearAccScale=<0.5 0.5 0.5>
    componentLinearVelScale=<0.5 0.5 0.5>
    overrideWorldGravity {z=-980}
   }
   n3 = RigidBody(ComponentPose=@n2, OverrideWorldGravity=@n1) {
    bClampLinearTranslationLimitToRefPose=true
    bOverrideWorldGravity=true
    overridePhysicsAsset=&/F_MED_SharpFang_Physics1
    simulationSpace=BaseBoneSpace
    alphaBoolBlend {blendOption=Linear}
    baseBoneRef {boneName=lowerarm_l}
    componentLinearAccScale=<0.5 0.5 0.5>
    componentLinearVelScale=<0.5 0.5 0.5>
    overrideWorldGravity {z=-980}
   }
   n4 = RigidBody(ComponentPose=@n3, OverrideWorldGravity="0,0,-980") {
    bClampLinearTranslationLimitToRefPose=true
    bOverrideWorldGravity=true
    overridePhysicsAsset=&/F_MED_SharpFang_Physics2
    simulationSpace=BaseBoneSpace
    alphaBoolBlend {blendOption=Linear}
    baseBoneRef {boneName=spine_05}
    componentLinearAccScale=<0.5 0.5 0.5>
    componentLinearVelScale=<0.5 0.5 0.5>
    overrideWorldGravity {z=-980}
   }
   n5 = Anim Dynamics - Bone: dyn_Sleeve_1_up_l(ComponentPose=@n4, Alpha=1, ExternalForce="0,0,0") {
    angularDampingOverride=1
    angularSpringConstant=5000
    bAngularSpring=true
    bGravityOverrideInSimSpace=true
    bOverrideAngularDamping=true
    gravityScale=0.35
    simulationSpace=RootRelative
    alphaBoolBlend {blendOption=Linear}
    alphaScaleBiasClamp {scale=0.2}
    boundBone {boneName=dyn_Sleeve_1_up_l}
    componentLinearAccScale {z=0.6}
    componentLinearVelScale {z=0.6}
    physicsBodyDefinitions [{
      collisionType=OuterSphere
      sphereCollisionRadius=4
      boundBone {boneName=dyn_Sleeve_1_up_l}
      boxExtents=<10 20 20>
      constraintSetup {angularLimitsMax=<45 40 10>, angularLimitsMin=<-45 -40 -10>, angularTarget=<1 0 0>}
      localJointOffset=<7 0 -4.7292>
     }]
   }
   n6 = Anim Dynamics - Bone: dyn_Sleeve_1_down_l(ComponentPose=@n5, Alpha=1, ExternalForce="0,0,0") {
    angularDampingOverride=1
    angularSpringConstant=5000
    bAngularSpring=true
    bGravityOverrideInSimSpace=true
    bOverrideAngularDamping=true
    gravityScale=0.35
    simulationSpace=RootRelative
    alphaBoolBlend {blendOption=Linear}
    alphaScaleBiasClamp {scale=0.2}
    boundBone {boneName=dyn_Sleeve_1_down_l}
    componentLinearAccScale {z=0.6}
    componentLinearVelScale {z=0.6}
    physicsBodyDefinitions [{
      collisionType=OuterSphere
      sphereCollisionRadius=4
      boundBone {boneName=dyn_Sleeve_1_down_l}
      boxExtents=<10 20 20>
      constraintSetup {angularLimitsMax=<45 40 10>, angularLimitsMin=<-45 -40 -10>, angularTarget=<1 0 0>}
      localJointOffset=<7 0 6.8305>
     }]
   }
   n7 = Anim Dynamics - Bone: dyn_Sleeve_1_up_r(ComponentPose=@n6, Alpha=1, ExternalForce="0,0,0") {
    angularDampingOverride=1
    angularSpringConstant=5000
    bAngularSpring=true
    bGravityOverrideInSimSpace=true
    bOverrideAngularDamping=true
    gravityScale=0.35
    simulationSpace=RootRelative
    alphaBoolBlend {blendOption=Linear}
    alphaScaleBiasClamp {scale=0.2}
    boundBone {boneName=dyn_Sleeve_1_up_r}
    componentLinearAccScale {z=0.6}
    componentLinearVelScale {z=0.6}
    physicsBodyDefinitions [{
      collisionType=OuterSphere
      sphereCollisionRadius=4
      boundBone {boneName=dyn_Sleeve_1_up_r}
      boxExtents=<10 20 20>
      constraintSetup {angularLimitsMax=<45 40 10>, angularLimitsMin=<-45 -40 -10>, angularTarget=<1 0 0>}
      localJointOffset=<7 -0.4944 5.0649>
     }]
   }
   n8 = Anim Dynamics - Bone: dyn_Sleeve_1_down_r(ComponentPose=@n7, Alpha=1, ExternalForce="0,0,0") {
    angularDampingOverride=1
    angularSpringConstant=5000
    bAngularSpring=true
    bGravityOverrideInSimSpace=true
    bOverrideAngularDamping=true
    gravityScale=0.35
    simulationSpace=RootRelative
    alphaBoolBlend {blendOption=Linear}
    alphaScaleBiasClamp {scale=0.2}
    boundBone {boneName=dyn_Sleeve_1_down_r}
    componentLinearAccScale {z=0.6}
    componentLinearVelScale {z=0.6}
    physicsBodyDefinitions [{
      collisionType=OuterSphere
      sphereCollisionRadius=4
      boundBone {boneName=dyn_Sleeve_1_down_r}
      boxExtents=<10 20 20>
      constraintSetup {angularLimitsMax=<45 40 10>, angularLimitsMin=<-45 -40 -10>, angularTarget=<1 0 0>}
      localJointOffset=<7 0 -6.6739>
     }]
   }
   n9 = "Get corr  Sleeve  L"() -> ("corr  Sleeve  L":double)
   n10 = "Transform (Modify) Bone - Bone: dyn_Sleeve_1_up_l"(ComponentPose=@n8, Alpha=@n9, Translation="0,0,0", Rotation="-2.172,-0.004,25.146", Scale="1,1,1") {
    rotationMode=BMM_Additive
    rotationSpace=BCS_BoneSpace
    scaleMode=BMM_Additive
    scaleSpace=BCS_BoneSpace
    translationMode=BMM_Additive
    translationSpace=BCS_BoneSpace
    alphaBoolBlend {blendOption=Linear}
    boneToModify {boneName=dyn_Sleeve_1_up_l}
    rotation=<rot -2.172 -0.004 25.146>
   }
   n11 = "Transform (Modify) Bone - Bone: dyn_Sleeve_1_down_l"(ComponentPose=@n10, Alpha=@n9, Translation="0.000,-4.520,2.877", Rotation="-1.442,2.267,32.448", Scale="1,1,1") {
    rotationMode=BMM_Additive
    rotationSpace=BCS_BoneSpace
    scaleMode=BMM_Additive
    scaleSpace=BCS_BoneSpace
    translationMode=BMM_Additive
    translationSpace=BCS_BoneSpace
    alphaBoolBlend {blendOption=Linear}
    boneToModify {boneName=dyn_Sleeve_1_down_l}
    rotation=<rot -1.442 2.267 32.448>
    translation {y=-4.52, z=2.877}
   }
   n12 = "Get corr  Sleeve  R"() -> ("corr  Sleeve  R":double)
   n13 = "Transform (Modify) Bone - Bone: dyn_Sleeve_1_up_r"(ComponentPose=@n11, Alpha=@n12, Translation="1.252,1.644,1.518", Rotation="1.495,1.603,-43.014", Scale="1,1,1") {
    rotationMode=BMM_Additive
    rotationSpace=BCS_BoneSpace
    scaleMode=BMM_Additive
    scaleSpace=BCS_BoneSpace
    translationMode=BMM_Additive
    translationSpace=BCS_BoneSpace
    alphaBoolBlend {blendOption=Linear}
    boneToModify {boneName=dyn_Sleeve_1_up_r}
    rotation=<rot 1.495 1.603 -43.014>
    translation=<1.252 1.644 1.518>
   }
   n14 = "Transform (Modify) Bone - Bone: dyn_Sleeve_1_down_r"(ComponentPose=@n13, Alpha=@n12, Translation="-0.742,-2.756,-2.452", Rotation="-0.000,0.000,-20.056", Scale="1,1,1") {
    rotationMode=BMM_Additive
    rotationSpace=BCS_BoneSpace
    scaleMode=BMM_Additive
    scaleSpace=BCS_BoneSpace
    translationMode=BMM_Additive
    translationSpace=BCS_BoneSpace
    alphaBoolBlend {blendOption=Linear}
    boneToModify {boneName=dyn_Sleeve_1_down_r}
    rotation=<rot 0 0 -20.056>
    translation=<-0.742 -2.756 -2.452>
   }
   n15 = Get bIsCrouching()
   n16 = "Transform (Modify) Bone - Bone: dyn_bag"!(ComponentPose=@n14, bAlphaBoolEnabled:bool=@n15, Translation="0,0,0", Rotation="-9.109,2.476,-36.770", Scale="1,1,1") -> (Pose) {
    alphaInputType=Bool
    rotationMode=BMM_Additive
    rotationSpace=BCS_BoneSpace
    scaleMode=BMM_Additive
    scaleSpace=BCS_BoneSpace
    translationMode=BMM_Additive
    translationSpace=BCS_BoneSpace
    alphaBoolBlend {blendInTime=0.5, blendOption=Linear, blendOutTime=0.3}
    boneToModify {boneName=dyn_bag}
    rotation=<rot -9.109 2.476 -36.77>
   }
   n17 = "Transform (Modify) Bone - Bone: dyn_Skirt_side_A_01_l"!(ComponentPose=@n16, bAlphaBoolEnabled:bool=@n15, Translation="0,0,0", Rotation="6.607,-3.940,-0.390", Scale="1.000,1.000,1.1") -> (Pose) {
    alphaInputType=Bool
    rotationMode=BMM_Additive
    rotationSpace=BCS_BoneSpace
    scaleMode=BMM_Additive
    scaleSpace=BCS_BoneSpace
    translationMode=BMM_Additive
    translationSpace=BCS_BoneSpace
    alphaBoolBlend {blendInTime=0.02, blendOption=Linear, blendOutTime=0.4}
    boneToModify {boneName=dyn_Skirt_side_A_01_l}
    rotation=<rot 6.607 -3.9401 -0.39>
    scale {z=1.1}
   }
   n18 = "Transform (Modify) Bone - Bone: dyn_bag_dop"!(ComponentPose=@n17, bAlphaBoolEnabled:bool=@n15, Translation="-0.407,0.221,-0.126", Rotation="1.144,15.505,-61.578", Scale="1.000,1.000,1.1") -> (Pose) {
    alphaInputType=Bool
    rotationMode=BMM_Additive
    rotationSpace=BCS_BoneSpace
    scaleMode=BMM_Additive
    scaleSpace=BCS_BoneSpace
    translationMode=BMM_Additive
    translationSpace=BCS_BoneSpace
    alphaBoolBlend {blendInTime=0.02, blendOption=Linear, blendOutTime=0.4}
    boneToModify {boneName=dyn_bag_dop}
    rotation=<rot 1.144 15.505 -61.578>
    scale {z=1.1}
    translation=<-0.407 0.221 -0.126>
   }
   n19 = "Transform (Modify) Bone - Bone: dyn_Skirt_side_A_01_r"!(
    ComponentPose=@n18
    bAlphaBoolEnabled:bool="Get corr Crouching  Moving  DOp"() -> ("corr Crouching  Moving  DOp":bool)
    Translation="0.361,-3.299,-3.681"
    Rotation="5.203,20.659,15.421"
    Scale="1.000,1.000,1.1"
   ) -> (Pose) {
    alphaInputType=Bool
    rotationMode=BMM_Additive
    rotationSpace=BCS_BoneSpace
    scaleMode=BMM_Additive
    scaleSpace=BCS_BoneSpace
    translationMode=BMM_Additive
    translationSpace=BCS_BoneSpace
    alphaBoolBlend {blendInTime=0.02, blendOption=Linear, blendOutTime=0.4}
    boneToModify {boneName=dyn_Skirt_side_A_01_r}
    rotation=<rot 5.203 20.659 15.421>
    scale {z=1.1}
    translation=<0.361 -3.299 -3.681>
   }
   n20 = Get bIsCrouchSprinting()
   n21 = "Transform (Modify) Bone - Bone: dyn_Chain_Side_1"!(ComponentPose=@n19, bAlphaBoolEnabled:bool=@n20, Translation="0,0,-1", Rotation="0.031,-14.709,-0.013", Scale="1.000,1.000,1.1") -> (Pose) {
    alphaInputType=Bool
    rotationMode=BMM_Additive
    rotationSpace=BCS_BoneSpace
    scaleMode=BMM_Additive
    scaleSpace=BCS_BoneSpace
    translationMode=BMM_Additive
    translationSpace=BCS_BoneSpace
    alphaBoolBlend {blendInTime=0.1, blendOption=Linear, blendOutTime=0.4}
    boneToModify {boneName=dyn_Chain_Side_1}
    rotation=<rot 0.031 -14.709 -0.013>
    scale {z=1.1}
    translation {z=-1}
   }
   n22 = "Transform (Modify) Bone - Bone: dyn_Chain_Side_2"!(ComponentPose=@n21, bAlphaBoolEnabled:bool=@n20, Translation="0,0,0", Rotation="0.022,18.491,0.021", Scale="1,1,1") -> (Pose) {
    alphaInputType=Bool
    rotationMode=BMM_Additive
    rotationSpace=BCS_BoneSpace
    scaleMode=BMM_Additive
    scaleSpace=BCS_BoneSpace
    translationMode=BMM_Additive
    translationSpace=BCS_BoneSpace
    alphaBoolBlend {blendInTime=0.1, blendOption=Linear, blendOutTime=0.4}
    boneToModify {boneName=dyn_Chain_Side_2}
    rotation=<rot 0.022 18.491 0.021>
   }
   n23 = Get corr Crouching Move Right() -> (corr Crouching Move Right:bool)
   n24 = "Transform (Modify) Bone - Bone: dyn_Chain_Side_1"(ComponentPose=@n22, bAlphaBoolEnabled=@n23, Translation="-0.345,0.578,1.526", Rotation="16.848,-12.514,17.070", Scale="1.000,1.000,1.1") {
    alphaInputType=Bool
    rotationMode=BMM_Additive
    rotationSpace=BCS_BoneSpace
    scaleMode=BMM_Additive
    scaleSpace=BCS_BoneSpace
    translationMode=BMM_Additive
    translationSpace=BCS_BoneSpace
    alphaBoolBlend {blendInTime=0.1, blendOption=Linear, blendOutTime=0.4}
    boneToModify {boneName=dyn_Chain_Side_1}
    rotation=<rot 16.848 -12.514 17.07>
    scale {z=1.1}
    translation=<-0.345 0.578 1.526>
   }
   n25 = "Transform (Modify) Bone - Bone: dyn_Chain_Side_2"(ComponentPose=@n24, bAlphaBoolEnabled=@n23, Translation="0.000,-0.000,0.689", Rotation="5.270,21.742,5.244", Scale="1,1,1") {
    alphaInputType=Bool
    rotationMode=BMM_Additive
    rotationSpace=BCS_BoneSpace
    scaleMode=BMM_Additive
    scaleSpace=BCS_BoneSpace
    translationMode=BMM_Additive
    translationSpace=BCS_BoneSpace
    alphaBoolBlend {blendInTime=0.1, blendOption=Linear, blendOutTime=0.4}
    boneToModify {boneName=dyn_Chain_Side_2}
    rotation=<rot 5.27 21.742 5.244>
    translation {z=0.689}
   }
   n26 = "Get corr Crouching  Move Left"() -> ("corr Crouching  Move Left":bool)
   n27 = "Transform (Modify) Bone - Bone: dyn_Chain_Side_1"(ComponentPose=@n25, bAlphaBoolEnabled=@n26, Translation="-1.161,-1.057,1.604", Rotation="0.193,16.300,25.934", Scale="1.000,1.000,1.1") {
    alphaInputType=Bool
    rotationMode=BMM_Additive
    rotationSpace=BCS_BoneSpace
    scaleMode=BMM_Additive
    scaleSpace=BCS_BoneSpace
    translationMode=BMM_Additive
    translationSpace=BCS_BoneSpace
    alphaBoolBlend {blendInTime=0.1, blendOption=Linear, blendOutTime=0.4}
    boneToModify {boneName=dyn_Chain_Side_1}
    rotation=<rot 0.193 16.3 25.934>
    scale {z=1.1}
    translation=<-1.161 -1.057 1.604>
   }
   n28 = "Transform (Modify) Bone - Bone: dyn_Chain_Side_2"(ComponentPose=@n27, bAlphaBoolEnabled=@n26, Translation="-0.505,-0.423,0.082", Rotation="-9.494,-20.032,-20.240", Scale="1,1,1") {
    alphaInputType=Bool
    rotationMode=BMM_Additive
    rotationSpace=BCS_BoneSpace
    scaleMode=BMM_Additive
    scaleSpace=BCS_BoneSpace
    translationMode=BMM_Additive
    translationSpace=BCS_BoneSpace
    alphaBoolBlend {blendInTime=0.1, blendOption=Linear, blendOutTime=0.4}
    boneToModify {boneName=dyn_Chain_Side_2}
    rotation=<rot -9.4941 -20.032 -20.2399>
    translation=<-0.505 -0.423 0.082>
   }
   n29 = "Transform (Modify) Bone - Bone: dyn_Chain_Side_1"(ComponentPose=@n28, bAlphaBoolEnabled=Get bIsJumping(), Translation="0,0,0", Rotation="-13.296,-12.363,6.269", Scale="1.000,1.000,1.1") {
    alphaInputType=Bool
    rotationMode=BMM_Additive
    rotationSpace=BCS_BoneSpace
    scaleMode=BMM_Additive
    scaleSpace=BCS_BoneSpace
    translationMode=BMM_Additive
    translationSpace=BCS_BoneSpace
    alphaBoolBlend {blendInTime=0.1, blendOption=Linear, blendOutTime=0.4}
    boneToModify {boneName=dyn_Chain_Side_1}
    rotation=<rot -13.296 -12.363 6.269>
    scale {z=1.1}
   }
   n30 = "Transform (Modify) Bone - Bone: dyn_Chain_Side_2"!(ComponentPose=@n29, Alpha:float=Get rot() -> (rot:double), Translation="0,0,0", Rotation="-0.000,-11.646,0.000", Scale="1,1,1") -> (Pose) {
    rotationMode=BMM_Additive
    rotationSpace=BCS_BoneSpace
    scaleMode=BMM_Additive
    scaleSpace=BCS_BoneSpace
    translationMode=BMM_Additive
    translationSpace=BCS_BoneSpace
    alphaBoolBlend {blendInTime=0.1, blendOption=Linear, blendOutTime=0.4}
    alphaScaleBiasClamp {bClampResult=true, bInterpResult=true}
    boneToModify {boneName=dyn_Chain_Side_2}
    rotation=<rot 0 -11.646 0.0001>
   }
   n31 = "Transform (Modify) Bone - Bone: extra_pel_scale"!(ComponentPose=@n30, bAlphaBoolEnabled:bool=Get lod2() -> (lod2:bool), Translation="0,0,0", Rotation="0,0,0", Scale="0,0,0") -> (Pose) {
    alphaInputType=Bool
    rotationMode=BMM_Additive
    rotationSpace=BCS_BoneSpace
    scaleMode=BMM_Additive
    scaleSpace=BCS_BoneSpace
    translationMode=BMM_Additive
    translationSpace=BCS_BoneSpace
    alphaBoolBlend {blendInTime=0.1, blendOption=Linear, blendOutTime=0.4}
    boneToModify {boneName=extra_pel_scale}
    scale=<0 0 0>
   }
   n32 = "Transform (Modify) Bone - Bone: dyn_Chain_Side_1"(
    ComponentPose=@n31
    bAlphaBoolEnabled=NOT Boolean(A=Get bRigidBodyEnabled())
    Translation="0,0,0"
    Rotation="-0.000,-5.927,0.000"
    Scale="1,1,1"
   ) {
    alphaInputType=Bool
    rotationMode=BMM_Additive
    rotationSpace=BCS_BoneSpace
    scaleMode=BMM_Additive
    scaleSpace=BCS_BoneSpace
    translationMode=BMM_Additive
    translationSpace=BCS_BoneSpace
    alphaBoolBlend {blendInTime=0.1, blendOption=Linear, blendOutTime=0.4}
    boneToModify {boneName=dyn_Chain_Side_1}
    rotation=<rot 0 -5.927 0>
   }
   n33 = "Transform (Modify) Bone - Bone: dyn_Chain_Side_3"!(ComponentPose=@n32, bAlphaBoolEnabled:bool=True, Translation="0,0,0", Rotation="-0.000,-18.284,0.000", Scale="1,1,1") -> (Pose) {
    alphaInputType=Bool
    rotationMode=BMM_Additive
    rotationSpace=BCS_BoneSpace
    scaleMode=BMM_Additive
    scaleSpace=BCS_BoneSpace
    translationMode=BMM_Additive
    translationSpace=BCS_BoneSpace
    alphaBoolBlend {blendInTime=0.1, blendOption=Linear, blendOutTime=0.4}
    boneToModify {boneName=dyn_Chain_Side_3}
    rotation=<rot 0 -18.284 0>
   }
   n34 = Control Rig(Source=Component To Local(ComponentPose=@n33)) {
    alphaBoolBlend {blendOption=Linear}
    controlRigAssetReference {blueprintRigClass=&/F_MED_SharpFang_ControlRig.F_MED_SharpFang_ControlRig_C}
   }
   Output Pose(Result=@n34) {layerGroup=None}
  ]}
 EventGraph {nodes [
   Event BlueprintUpdateAnimation()
   > n1 = Sequence!(execute) -> (then_0, then_1, then_2:Exec)
   n2 = Get bIsMoving2D()
   n3 = Get bIsCrouching()
   > Set is Crouching Moving(is Crouching Moving:bool=AND Boolean(A=@n3, B=NOT Boolean(A=@n2))) -> (Output_Get:bool)
   > n4 = Branch(Condition=Get bIsFalling())
   > Set Gravity Vector(Gravity Vector:Vector="0, 0,300") -> (Output_Get:Vector)
   Set Gravity Vector(execute=@n4.else, Gravity Vector="0, 0,-980")
   n5 = Enum to String!(Enumerator:EFortCardinalDirection Enum=GetCardinalDirectionFromAngle!(
     self:Fort Anim Instance Object Reference
     Angle:float=Get PawnMovementDirectionAzimuth() -> (PawnMovementDirectionAzimuth:float)
     DeadZoneAngle:float=0
    ) -> (ReturnValue:EFortCardinalDirection Enum=North)) -> (ReturnValue:String)
   n6 = "Equal Exactly (String)"(A=@n5, B=West)
   n7 = "Equal Exactly (String)"(A=@n5, B=East)
   n8 = OR Boolean(A=@n6, B=@n7)
   n9 = AND Boolean(A=@n3, B=@n2)
   "Set corr Crouching  Moving  DOp"(execute=@n1.then_2, "corr Crouching  Moving  DOp":bool=AND Boolean(A=@n9, B=NOT Boolean(A=@n8))) -> (Output_Get:bool)
   > "Set corr Crouching  Move Left"("corr Crouching  Move Left":bool=AND Boolean(A=@n9, B=@n6)) -> (Output_Get:bool)
   > Set corr Crouching Move Right(corr Crouching Move Right:bool=AND Boolean(A=@n9, B=@n7)) -> (Output_Get:bool)
   > "Set corr Crouching  Move Back"("corr Crouching  Move Back":bool=AND Boolean(A=@n9, B="Equal Exactly (String)"(A=@n5, B=South))) -> (Output_Get:bool)
   n10 = Get bIsSprinting()
   > Set moveWithoutSprint(moveWithoutSprint:bool=AND Boolean(A=Get bIsMoving2D(), B=NOT Boolean(A=@n10))) -> (Output_Get:bool)
   n11 = GetOwnerSkeletalMeshComponent()
   "Set corr  Skirt  Back  L"(
    execute=@n1.then_1
    "corr  Skirt  Back  L":double=MapRangeClamped(Value=GetTwistAndSwingAngleOfDeltaRotationFromRefPose(self=@n11, BoneName=thigh_r).OutSwingAngle, InRangeA=170, InRangeB=140)
   ) -> (Output_Get:double)
   > "Set corr  Skirt  Back  R"("corr  Skirt  Back  R":double=MapRangeClamped(Value=GetTwistAndSwingAngleOfDeltaRotationFromRefPose(self=@n11, BoneName=thigh_l).OutSwingAngle, InRangeA=25, InRangeB=100)) -> (Output_Get:double)
   > "Set corr  Sleeve  L"("corr  Sleeve  L":double=MapRangeClamped(Value=GetTwistAndSwingAngleOfDeltaRotationFromRefPose(self=@n11, BoneName=upperarm_twist_01_l).OutSwingAngle, InRangeA=5, InRangeB=20)) -> (Output_Get:double)
   > "Set corr  Sleeve  R"("corr  Sleeve  R":double=MapRangeClamped(Value=GetTwistAndSwingAngleOfDeltaRotationFromRefPose(self=@n11, BoneName=upperarm_twist_01_r).OutSwingAngle, InRangeA=5, InRangeB=20)) -> (Output_Get:double)
   > Set lod2(lod2:bool="integer > integer"(A=Get CurrentLOD(), B=1)) -> (Output_Get:bool)
   > n12 = Branch(Condition=Get is Crouching Moving() -> (is Crouching Moving:bool))
   > Set rot(rot:double="Absolute (Float)"(A=Get!() -> (PawnDeltaRotation_Roll:float, PawnDeltaRotation_Pitch:float, PawnDeltaRotation_Yaw:float).PawnDeltaRotation_Yaw)) -> (Output_Get:double)
   Set rot(execute=@n12.else, rot=0)
   Gravity Override State Function DEPRECATED(
    CharacterPartAnimInstance=Self-Reference()
    GameTimeInSeconds=GetGameTimeInSeconds()
    GravityOverrideParams=Get Gravity override Params() -> (Gravity override Params:Gravity Override Params Struct Structure)
   )
   Set Gravity Vector(Gravity Vector="0, 0, 0")
  ]}
]
variables [
 "corr  Elbow  L" {type=number}
 "corr  Elbow  R" {type=number}
 "corr  Knee  L" {type=number}
 "corr  Knee  R" {type=number}
 "corr  Leg  Back  L" {type=number}
 "corr  Leg  Back  R" {type=number}
 "corr  Scapula  fr L" {type=number}
 "corr  Scapula  fr R" {type=number}
 "corr  Skirt  Back  L" {type=number}
 "corr  Skirt  Back  R" {type=number}
 "corr  Sleeve  L" {type=number}
 "corr  Sleeve  R" {type=number}
 "corr  Wrist  up  L" {type=number}
 "corr  Wrist  up  R" {type=number}
 "corr Crouching  Move Back" {type=boolean}
 "corr Crouching  Move Left" {type=boolean}
 "corr Crouching  Moving  DOp" {type=boolean}
 corr Crouching Move Right {type=boolean}
 Gravity override Params {type=GravityOverrideParamsStruct}
 Gravity Vector {type=Vector}
 is Crouching Moving {type=boolean}
 lod2 {type=boolean, default=true}
 moveWithoutSprint {type=boolean}
 rot {type=number}
]
