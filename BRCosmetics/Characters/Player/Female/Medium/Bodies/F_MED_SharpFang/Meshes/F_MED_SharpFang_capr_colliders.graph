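# Physics asset with simplified body colliders for the F_MED_SharpFang female medium player body, named as a "capr" collider set.
# One capsule per bone covers pelvis, two spine bones, neck, upper and lower arms, thighs, calves and feet, with no hand or head shapes.
# Constraints with uniform swing and twist limit angles chain limbs to pelvis and upper spine, but neck_01 and lowerarm_r have none.
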PhysicsAsset /BRCosmetics/Characters/Player/Female/Medium/Bodies/F_MED_SharpFang/Meshes/F_MED_SharpFang_capr_colliders
bodies {
 calf_l [{shapeType=Capsule, length=19.9437, radius=12.2196, center=<-21.8883 -0.3062 -0.4068>, rotation=<rot 0 92.0613 -88.6744>}]
 calf_r [{shapeType=Capsule, length=19.9437, radius=12.2196, center=<21.8883 0.3063 0.4068>, rotation=<rot 0 92.0614 -88.6744>}]
 foot_l [{shapeType=Capsule, length=16.1747, radius=8.1369, center=<-2.4132 7.2567 0.11>, rotation=<rot 0 -5.2063 -87.5509>}]
 foot_r [{shapeType=Capsule, length=16.1747, radius=8.1369, center=<2.4132 -7.2566 -0.1099>, rotation=<rot 0 -5.2063 -87.5509>}]
 lowerarm_l [{shapeType=Capsule, length=21.4029, radius=7.2059, center=<13.8204 -0.1711 -0.2324>, rotation=<rot 0.0442 87.5888 -89.3428>}]
 lowerarm_r [{
   shapeType=Capsule
   length=21.7473
   radius=8.2526
   shapeName=lowerarm_r_capsule
   center=<-14.2851 0 -0.1144>
   rotation=<rot 84.9317 -179.9996 -175.3167>
  }]
 neck_01 [{shapeType=Capsule, length=20.7159, radius=9.2841, shapeName=neck_01_capsule, center=<-3.9246 5.413 1.0045>}]
 pelvis [{shapeType=Capsule, length=14.8098, radius=15.6262, center=<-8.0085 2.0899 0.1159>, rotation=<rot -1.574 -91.3406 1.1372>}]
 spine_02 [{shapeType=Capsule, length=14.8428, radius=14.1827, center=<-6.0699 1.1022 0.317>, rotation=<rot 2.4975 -90 -0.8397>}]
 spine_05 [{shapeType=Capsule, length=13.5662, radius=11.7674, center=<-9.706 0.6709 0.8294>, rotation=<rot -5.4587 -90.0001 -0.4094>}]
 thigh_l [{shapeType=Capsule, length=23.8229, radius=11.5272, center=<-21.2267 -1.0137 1.5148>, rotation=<rot 0 -91.391 -87.717>}]
 thigh_r [{shapeType=Capsule, length=23.6856, radius=11.3752, center=<20.9248 -0.2624 -1.7754>, rotation=<rot -0.7613 -92.7535 -85.5143>}]
 upperarm_l [{shapeType=Capsule, length=18.5571, radius=7.7569, center=<13.0161 -0.2552 -0.4919>, rotation=<rot 0 87.0384 -89.7145>}]
 upperarm_r [{shapeType=Capsule, length=22.1421, radius=7.8164, center=<-12.5457 0.0283 0.851>, rotation=<rot 0 102.6178 -86.945>}]
}
constraints [
 {bone1Name=spine_02, bone2Name=pelvis, swing1LimitDegrees=45, swing2LimitDegrees=45, twistLimitDegrees=45}
 {bone1Name=spine_05, bone2Name=spine_02, swing1LimitDegrees=45, swing2LimitDegrees=45, twistLimitDegrees=45}
 {bone1Name=upperarm_l, bone2Name=spine_05, swing1LimitDegrees=45, swing2LimitDegrees=45, twistLimitDegrees=45}
 {bone1Name=lowerarm_l, bone2Name=upperarm_l, swing1LimitDegrees=45, swing2LimitDegrees=45, twistLimitDegrees=45}
 {bone1Name=upperarm_r, bone2Name=spine_05, swing1LimitDegrees=45, swing2LimitDegrees=45, twistLimitDegrees=45}
 {bone1Name=thigh_l, bone2Name=pelvis, swing1LimitDegrees=45, swing2LimitDegrees=45, twistLimitDegrees=45}
 {bone1Name=calf_l, bone2Name=thigh_l, swing1LimitDegrees=45, swing2LimitDegrees=45, twistLimitDegrees=45}
 {bone1Name=foot_l, bone2Name=calf_l, swing1LimitDegrees=45, swing2LimitDegrees=45, twistLimitDegrees=45}
 {bone1Name=thigh_r, bone2Name=pelvis, swing1LimitDegrees=45, swing2LimitDegrees=45, twistLimitDegrees=45}
 {bone1Name=calf_r, bone2Name=thigh_r, swing1LimitDegrees=45, swing2LimitDegrees=45, twistLimitDegrees=45}
 {bone1Name=foot_r, bone2Name=calf_r, swing1LimitDegrees=45, swing2LimitDegrees=45, twistLimitDegrees=45}
]
